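# Unlit masked Niagara particle material for the BlackMonday emote effects, serving several looks from one graph.
# Scalar-parameter lerps select a masked texture, procedural gradient shape, panning speed lines, water-style flash flipbook or sparkle texture, for both color and mask.
# Dynamic parameters drive noise-thresholded sphere dissolve, center cutout, panning and flipbook phase.
# Particle color tints most looks, with optional time-of-day tinting before a shared emissive-adjustment function.
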Material /BRCosmetics/Athena/Effects/Materials/Emotes/M_BlackMonday_Emote01
graph {
 properties {
  bUsedWithNiagaraMeshParticles=true
  bUsedWithNiagaraRibbons=true
  bUsedWithNiagaraSprites=true
  blendMode=BLEND_Masked
  pixelDepthOffsetMode=PDOM_Legacy
  shadingModel=MSM_Unlit
  usedShadingModels=Unlit
 }
 nodes [
  n1 = ParticleColor()
  n2 = TextureCoordinate()
  n3 = LinearInterpolate(
   A=@n2
   B=MaterialFunctionCall(XY=@n2) {materialFunction=&Swizzle}.YX
   Alpha=ScalarParameter() {parameterName=UseSwizzleUV}
  )
  n4 = LinearInterpolate(Alpha=ComponentMask(None=@n3) {r=true}) {constA=0.5, constB=0.56}
  n5 = LinearInterpolate(Alpha=ComponentMask(None=@n3) {g=true}) {constA=0.5}
  n6 = TextureSampleParameter2D(UVs=StaticSwitchParameter(True=AppendVector(A=@n4, B=@n5), False=TextureCoordinate())) {parameterName=Texture, texture=/Game/Effects/Fort_Effects/Textures/T_Works}
  n7 = AppendVector(A=@n6.RGB, B=@n6.RGB)
  n8 = DynamicParameter() {defaultValue=<rgba 1 0.5 1 1>, paramNames [Param1,SphereMask,Param3,Param4]}
  n9 = Multiply(
   A=SphereMask(A=TextureCoordinate(), B=Constant2Vector() {g=1, r=0.5}, Radius=LinearInterpolate(Alpha=@n8.Param1) {constB=3}) {attenuationRadius=1, hardnessPercent=0}
   B=TextureSample(UVs=TextureCoordinate() {uTiling=0.4, vTiling=0.25}) {texture=/Game/Global/Textures/Default/Noise/TankMoteNoise}.R
  )
  n10 = LinearInterpolate(Alpha=OneMinus(None=@n8.Param1)) {constA=-0.1}
  n11 = Constant() {r=1}
  n12 = Constant()
  n13 = Multiply(
   A=@n1.A
   B=Multiply(
    A=StaticComponentMaskParameter(None=@n7) {defaultB=true, parameterName=Maks_RGBA}
    B=If(A=@n9, B=@n10, "A > B"=@n11, "A == B"=@n12, "A < B"=@n12)
   )
  )
  n14 = TextureCoordinate()
  n15 = Constant() {r=0.5}
  n16 = DynamicParameter() {defaultValue=<rgba 1 0.5 1 1>, paramNames [Param1,SphereMask,Param3,Param4]}
  n17 = Multiply(A=@n13, B=OneMinus(None=SphereMask(A=@n14, B=@n15, Radius=@n16.SphereMask) {attenuationRadius=0.5}))
  n18 = ScalarParameter() {parameterName=SphereMask}
  n19 = ComponentMask(None=TextureCoordinate()) {r=true}
  n20 = OneMinus(None=@n19)
  n21 = Multiply(A=@n20, B=@n20)
  n22 = SphereMask(A=TextureCoordinate(), B=Constant() {r=0.5}, Radius=Constant() {r=10}, Hardness=Constant())
  n23 = Multiply(A=Multiply(A=Multiply(A=@n21, B=@n19)) {constB=3}, B=Multiply(A=OneMinus(None=@n22), B=Constant() {r=5}))
  n24 = ParticleColor()
  n25 = ScalarParameter() {parameterName=HighTapas}
  n26 = Constant() {r=1}
  n27 = Constant()
  n28 = TextureSample(UVs=Panner(Coordinate=TextureCoordinate() {uTiling=0.1}, Time=DynamicParameter().Param1, Speed=AppendVector(A=@n26, B=@n27)) {speedX=1}) {samplerSource=SSM_Clamp_WorldGroupSettings, texture=/Game/Effects/Fort_Effects/Emotes/BuffCatComic/P_Emote_BuffCatComic_SpeedLine}
  n29 = TextureCoordinate()
  n30 = OneMinus(None=Cosine(None=ComponentMask(None=@n29) {r=true}))
  n31 = TextureSample(UVs=MaterialFunctionCall("Animation  Phase (0-1)"=DynamicParameter().Param1) {materialFunction=&FlipBook}.UVs) {texture=/Game/Effects/Fort_Effects/Emotes/BuffCatComic/P_Emote_BuffCatComic_Flash}
  n32 = ScalarParameter() {parameterName=UseforWater}
  n33 = ScalarParameter() {parameterName=SpeedLine}
  n34 = LinearInterpolate(
   A=LinearInterpolate(A=LinearInterpolate(A=@n13, B=@n17, Alpha=@n18), B=Multiply(A=@n23, B=@n24.A), Alpha=@n25)
   B=LinearInterpolate(A=Multiply(A=@n28.R, B=@n30), B=@n31.G, Alpha=@n32)
   Alpha=@n33
  )
  n35 = ScalarParameter() {parameterName=Sparkly}
  n36 = TextureSample!(UVs, Tex, Bias, Apply View MipBias) {mipValueMode=TMVM_MipBias, texture=/Game/Effects/Fort_Effects/Emotes/BuffCatComic/P_Emote_BuffCatComic_Sparkly}
  LinearInterpolate(A=@n34, B=@n36.A, Alpha=@n35)
  LinearInterpolate(A=@n34, B=@n36.A, Alpha=ScalarParameter() {parameterName=Sparkly 2})
  n37 = Multiply(A=VectorParameter() {parameterName=Color, defaultValue=<rgba 1.8 0.6025 0 1>}.RGB, B=@n1.RGB)
  n38 = Multiply(A=@n23, B=@n24.RGB)
  n39 = ComponentMask(None=TextureCoordinate() {uTiling=0.1}) {g=true}
  n40 = LinearInterpolate(Alpha=Round(None=Power(Base=@n39) {constExponent=1.2}))
  n41 = Multiply(A=@n36.R, B=VectorParameter() {parameterName=BaseColor, defaultValue=<rgba 0 0 0 1>}.RGB)
  n42 = Add(A=@n41, B=Multiply(A=@n36.B) {constB=10})
  n43 = ParticleColor()
  n44 = LinearInterpolate(
   A=LinearInterpolate(A=LinearInterpolate(A=@n37, B=@n38, Alpha=@n25), B=LinearInterpolate(A=@n40, B=@n31.B, Alpha=@n32), Alpha=@n33)
   B=LinearInterpolate(A=@n42, B=Multiply(A=@n42, B=@n43.RGB), Alpha=ScalarParameter() {parameterName=Use Particle Color Sparkle})
   Alpha=@n35
  )
  n45 = MaterialFunctionCall(
   Color_Morning=VectorParameter() {group=ToD, parameterName=MorningTint, defaultValue=<rgba 1 1 1 1>}.RGB
   Color_Day=VectorParameter() {group=ToD, parameterName=DayTint, defaultValue=<rgba 1 1 1 1>}.RGB
   Color_Evening=VectorParameter() {group=ToD, parameterName=EveningTint, defaultValue=<rgba 0.75 0.75 0.75 1>}.RGB
   Color_Night=VectorParameter() {group=ToD, parameterName=NightTint, defaultValue=<rgba 0.5 0.5 0.5 1>}.RGB
  ) {materialFunction=&Color_Time_of_Day}
  MaterialFunctionCall(Emissive=LinearInterpolate(A=@n44, B=Multiply(A=@n45.Result, B=@n44), Alpha=ScalarParameter() {group=ToD, parameterName=Use ToD Multi})) {materialFunction=&MF_FX_EmissiveAdjustments}
  OneMinus(None=Sine(None=@n39))
  Subtract(A=@n41, B=Multiply(A=@n36.G))
 ]
}
