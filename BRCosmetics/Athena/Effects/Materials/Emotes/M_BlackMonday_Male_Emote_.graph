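# Translucent unlit particle material, named for the BlackMonday male emote, that draws an animated flipbook sprite with an erosion dissolve.
# Dynamic parameters pick the flipbook frame and drive a noise-perturbed sphere mask, thresholded by particle green to cut the sprite and darken its edge.
# A yellow-to-orange gradient scaled by particle red tints the emissive.
# A "Use HolidayCookies Skydive FX" switch substitutes zero emissive and a four-sphere-mask dissolve, alongside gingerbread backpack texture samples.
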Material /BRCosmetics/Athena/Effects/Materials/Emotes/M_BlackMonday_Male_Emote_
graph {
 properties {
  bEnableResponsiveAA=true
  bUsedWithNiagaraMeshParticles=true
  blendMode=BLEND_Translucent
  pixelDepthOffsetMode=PDOM_Legacy
  shadingModel=MSM_Unlit
  usedShadingModels=Unlit
 }
 nodes [
  n1 = Time()
  n2 = Add(A=Sine(None=@n1))
  n3 = Add(A=Sine(None=Add(A=@n1) {constB=0.5}))
  AppendVector(A=Round(None=Divide(A=@n2)), B=Round(None=Divide(A=@n3)))
  LinearInterpolate(Alpha=MaterialFunctionCall() {materialFunction=&DebugTimeSine}.Result) {constB=1.5}
  n4 = LinearInterpolate(
   A=Constant3Vector() {constant=<rgba 1 0.9536 0.5604 0>}
   B=Constant3Vector() {constant=<rgba 1 0.5225 0 0>}
   Alpha=ComponentMask(None=TextureCoordinate()) {g=true}
  )
  n5 = Constant()
  n6 = ParticleColor()
  n7 = DynamicParameter() {defaultValue=<rgba 7 1 1 1>}
  n8 = Add(A=@n7.Param1) {constB=0.5}
  n9 = TextureSample(UVs=MaterialFunctionCall(
    "Animation  Phase (0-1)"=Divide(A=@n8) {constB=12}
    Number of Rows=Constant() {r=3}
    Number of Columns=Constant() {r=4}
    UVs=TextureCoordinate()
   ) {materialFunction=&FlipBook}.UVs) {texture=/Game/Effects/Fort_Effects/Textures/T_Works}
  n10 = Multiply(A=Multiply(A=Desaturation(None=@n4, Fraction=@n5), B=@n6.R), B=@n9.G)
  n11 = Saturate(None=@n7.Param2)
  n12 = SphereMask(A=TextureCoordinate(), B=Round(None=@n11)) {attenuationRadius=1.3, hardnessPercent=20}
  n13 = Time()
  n14 = TextureSample(UVs=Panner(Coordinate=TextureCoordinate() {uTiling=0.28, vTiling=0.4}, Time=Add(A=@n11, B=@n13)) {speedX=0.1, speedY=0.3}) {texture=/Game/Global/Textures/Default/Noise/TankMoteNoise}
  n15 = Clamp(None=Add(A=@n12, B=Multiply(A=@n12, B=@n14.R)))
  n16 = Constant()
  n17 = Constant() {r=1}
  n18 = If(A=@n15, B=Add(A=@n6.G) {constB=0.055}, "A > B"=@n16, "A == B"=@n17, "A < B"=@n17)
  n19 = Clamp(None=Subtract(A=@n18, B=If(A=@n15, B=@n6.G, "A > B"=@n16, "A == B"=@n17, "A < B"=@n17)))
  MaterialFunctionCall(Emissive=StaticSwitchParameter(True=Constant(), False=Subtract(A=@n10, B=@n19)) {parameterName=Use HolidayCookies Skydive FX}) {materialFunction=&MF_FX_EmissiveAdjustments}
  n20 = TextureSampleParameter2D() {parameterName=BaseColor, texture=/BRCosmetics/Accessories/FORT_Backpacks/F_MED_Gingerbread_Backpack/Textures/T_Female_Gingerbread_Backpack_D}
  n21 = ScalarParameter() {parameterName=Desaturation}
  Reroute(None=Multiply(A=ParticleColor().RGB, B=Desaturation(None=@n20.RGB, Fraction=@n21)))
  n22 = TextureSampleParameter2D() {
   parameterName=Secular
   samplerType=SAMPLERTYPE_Masks
   texture=/BRCosmetics/Accessories/FORT_Backpacks/F_MED_Gingerbread_Backpack/Textures/T_Female_Gingerbread_Backpack_S
  }
  Reroute(None=@n22.RGB)
  n23 = SphereMask(
   A=TextureCoordinate() {coordinateIndex=1}
   B=Constant2Vector() {g=0.8, r=0.2}
   Radius=ScalarParameter() {defaultValue=0.33, parameterName=Radius02}
  ) {attenuationRadius=0.33}
  n24 = VectorParameter() {parameterName=Mask_location01, defaultValue=<rgba 0.5 0.145 0 1>}
  n25 = SphereMask(
   A=TextureCoordinate() {coordinateIndex=1}
   B=AppendVector(A=@n24.R, B=@n24.R)
   Radius=ScalarParameter() {defaultValue=0.35, parameterName=Radius01}
  ) {attenuationRadius=0.35}
  n26 = TextureCoordinate() {coordinateIndex=1}
  n27 = Constant2Vector() {g=0.9, r=0.7}
  n28 = ScalarParameter() {defaultValue=0.2, parameterName=Radius03}
  n29 = Add(
   A=Add(A=Multiply(A=@n23) {constB=0.5}, B=Multiply(A=@n25))
   B=Multiply(A=SphereMask(A=@n26, B=@n27, Radius=@n28) {attenuationRadius=0.2}) {constB=0.75}
  )
  n30 = Multiply(A=SphereMask(
    A=TextureCoordinate() {coordinateIndex=1}
    B=Constant2Vector() {g=0.5, r=0.9}
    Radius=ScalarParameter() {defaultValue=0.25, parameterName=Radius04}
   ) {attenuationRadius=0.25}) {constB=0.25}
  n31 = Constant()
  n32 = If(
   A=DynamicParameter() {defaultValue=<rgba 1.5 1 1 1>, paramNames [Dissolve,Param2,Param3,Param4]}.Dissolve
   B=Saturate(None=Add(A=@n29, B=@n30))
   "A > B"=Constant() {r=1}
   "A == B"=@n31
   "A < B"=@n31
  )
  Reroute(None=@n32)
  n33 = TextureSampleParameter2D() {
   parameterName=Normal
   samplerType=SAMPLERTYPE_Normal
   texture=/BRCosmetics/Accessories/FORT_Backpacks/F_MED_Gingerbread_Backpack/Textures/T_Female_Gingerbread_Backpack_N
  }
  Reroute(None=@n33.RGB)
  StaticSwitchParameter(True=@n32, False=Multiply(A=Add(A=@n9.R, B=@n9.R), B=@n18)) {parameterName=Use HolidayCookies Skydive FX}
  n34 = TextureCoordinate() {uTiling=0.48}
  n35 = Add(B=Time())
  n36 = TextureCoordinate()
  Subtract(
   A=TextureSample(UVs=Panner(Coordinate=@n34, Time=@n35) {speedX=0.5}) {texture=/Game/Effects/Fort_Effects/Textures/Outlander/T_InTheZone_LightRays_01}.R
   B=TextureSample(UVs=Panner(Coordinate=@n36) {speedX=-0.4}) {texture=&T_InTheZone_LightRays_01}.G
  )
 ]
}
